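# Niagara system for the CocoaSmoore vehicle boost cosmetic, trailing cocoa-colored ribbons and sprites from filtered skeletal mesh sockets.
# Ribbons spawn per port, with quality-based rates and growing width curves feeding two ribbon renderers.
# One sprite emitter makes marshmallow, spark and smoke looks through three spawn groups, switching motion and lifetime by group and writing it as the renderer visibility tag.
# Both emitters inherit source movement; painted-variant overrides are disabled.
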NiagaraSystem /VehicleCosmetics/Boosts/Boost_CocoaSmoore/NS/NS_Boost_Cocoasmoore
dependencies {
 usedDataInterfaces [
  &NiagaraDataInterfaceCurve
  &NiagaraDataInterfaceSpriteRendererInfo
  &NiagaraDataInterfaceSkeletalMesh
 ]
 usedDynamicInputs [&RandomRangeVector]
 usedModules [
  &self:SystemSpawnScript.SystemScriptSource.SystemScriptGraph.NiagaraNodeAssignment_0.SetVariables_#1
  &SystemState
 ]
}
emitters {
 Ribbons {
  emitterUpdateScript [
   &EmitterState(Inactive Response=::NewEnumerator1, Life Cycle Mode=::NewEnumerator1, Loop Duration=1, UseLoopDelay=false)
   &VC_FilteredSocketCount()
   &VC_RateQualityHandler(Epic Spawn Rate=45, Low Spawn Rate=30, Use Custom Rate?=true)
   &VC_SpawnRibbon_v1(Number of Ribbons=$Emitter.PortCount, SpawnRate=60)
  ]
  particleSpawnScript [
   &VC_PaintedVariantsOverride(
    Black=<rgba 0.01 0.01 0.01 2>
    BurntSienna=<rgba 0.15 0.04 0 1>
    Cobalt=<rgba 0.006 0.0256 0.3 1.5>
    Crimson=<rgba 0.0816 0 0.0022 1.5>
    ForestGreen=<rgba 0 0.0608 0.0031 1.5>
    Grey=<rgba 0.06 0.06 0.06 1>
    Lime=<rgba 0.0334 0.1565 0.025 1.5>
    Orange=<rgba 0.2205 0.0372 0 1.5>
    Pink=<rgba 0.3003 0.0771 0.2854 1>
    Purple=<rgba 0.015 0 0.09 2>
    R1 Black:LinearColor=<rgba 0.03 0.03 0.03 1>
    R1 BurntSienna:LinearColor=<rgba 0.08 0.0133 0 1>
    R1 Cobalt:LinearColor=<rgba 0.045 0.1057 0.45 1>
    R1 ForestGreen:LinearColor=<rgba 0 0.2319 0.004 1>
    R1 Grey:LinearColor=<rgba 0.12 0.12 0.12 1>
    R1 Orange:LinearColor=<rgba 0.6 0.08 0 1>
    R1 Pink:LinearColor=<rgba 0.32 0.112 0.32 1>
    R1 Purple:LinearColor=<rgba 0.035 0 0.3 1>
    R1 Saffron:LinearColor=<rgba 0.5 0.2917 0 1>
    R1 SkyBlue:LinearColor=<rgba 0.1226 0.2854 0.5 1>
    R1 TitaniumWhite:LinearColor=<rgba 0.45 0.45 0.45 1>
    Saffron=<rgba 0.12 0.08 0 1.25>
    SkyBlue=<rgba 0 0.0625 0.25 1.5>
    TitaniumWhite=<rgba 0.1406 0.1406 0.1406 1>
    Use R1 Painted Overrides=true
   ) {enabled=false}
   &VC_InitializeRibbon_v1(
    Color=<rgba 0.1302 0.0706 0.054 1>
    Custom Offset=true
    Enable Skeletal Position=true
    Lifetime=0.9
    Local Vector Offset=&AddVectorToPosition(Position=<0 0 0>, Vector=$User.DTCarOffset)
    Offset Amount=1
    Write Material Random=false
   )
   &AddVelocity(Velocity=<1750 0 0>, Velocity Speed Scale=-1) {module=AddVelocity001}
  ]
  particleUpdateScript [
   &ParticleState()
   &InheritSourceMovement(
    Applied Position Scale=&Multiply_VectorByFloat(Vector=<1 1 1>, Float=1)
    ApplyForce=false
    ApplyPosition=true
    ApplyPositionFalloff=false
   )
   &SolveForcesAndVelocity(Clamp Velocity=false, Limit Acceleration=false)
   &VC_UpdateRibbon_v1(Enable Skeletal Position=true)
   /SetVariables(
    Particles.Cocoa Size:NiagaraFloat=&FloatFromCurve(
     FloatCurve=DataInterface() {propertyValues {Curve {
        defaultValue=1
        keys [
         {arriveTangent=0.9441, interpMode=RCIM_Cubic, leaveTangent=0.9441, tangentMode=RCTM_User, time=0, value=0.5}
         {arriveTangent=1.1146, interpMode=RCIM_Cubic, leaveTangent=1.1146, tangentMode=RCTM_User, time=0.15, value=1}
         {arriveTangent=0.774, interpMode=RCIM_Cubic, leaveTangent=0.774, tangentMode=RCTM_User, time=1, value=1.5}
        ]
       }}}
     Scale Curve=80
    )
    Particles.Flame Coffe Size=80
   ) {isSetParameters=true}
  ]
  rendererClasses [&NiagaraRibbonRendererProperties]
  renderers [
   {rendererClass=&NiagaraRibbonRendererProperties}
   {rendererClass=&NiagaraRibbonRendererProperties, rendererIndex=1}
  ]
 }
 Sprites001 {
  emitterUpdateScript [
   &EmitterState()
   &VC_FilteredSocketCount()
   &SpawnRate(
    SpawnRate=&MergeHelper_FloatSwitch(0=55, 1=40, 2=150, 3=?, Index=$User.GameModeID, Num=::NewEnumerator1)
    Use Spawn Probability=false
   )
   &SpawnRate(SpawnGroup=1, SpawnRate=250, Use Spawn Probability=false) {module=SpawnRate001}
   &SpawnRate(SpawnGroup=2, SpawnRate=60, Use Spawn Probability=false) {module=SpawnRate002}
  ]
  particleSpawnScript [
   &VC_PaintedVariantsOverride() {enabled=false}
   &InitializeParticle(
    Color Mode=::NewEnumerator0
    Lifetime=&MergeHelper_FloatSwitch(0=0.42, 1=0.5, 2=1.25, 3=?, Index=$Emitter.SpawnGroup, Num=::NewEnumerator1)
    Sprite Rotation Mode=::NewEnumerator2
    Sprite Size Mode=::NewEnumerator3
    Uniform Sprite Size=25
    UsePositionOffset=true
    Write Material Random=true
   )
   &PartitionParticles(Number Of Partitions=$Emitter.PortCount)
   &SkeletalMeshLocation(
    Bone / Socket Index=$Particles.PartitionParticles.Partition
    Bone / Socket Sampling Mode=::NewEnumerator5
    MaskSpawnGroup=false
    Mesh Sampling Type=::NewEnumerator6
    Sampled Position Offset=&AddVectorToPosition(Position=<-30 0 0>, Vector=$User.DTCarOffset)
    Skeletal Mesh=$System.SkeletalMesh
   )
   &AddVelocity(
    Cone Angle=&MergeHelper_FloatSwitch(0=12, 1=0, 2=?, 3=?, Index=$Emitter.SpawnGroup, Num=::NewEnumerator0)
    Cone Axis=&MergeHelper_VectorSwitch(0=<1 0 0>, 1=<0 0 0>, 2=<-1 0 0.8>, 3=?, Index=$Emitter.SpawnGroup, Num=::NewEnumerator1)
    Module.UseConeFalloff=false
    Module.UseVelDistribution=false
    Override Seed=false
    Velocity Mode=::NewEnumerator2
    Velocity Speed=&MergeHelper_FloatSwitch(0=-1500, 1=-1500, 2=50, 3=?, Index=$Emitter.SpawnGroup, Num=::NewEnumerator1)
   )
   &ShapeLocation(
    MaskSpawnGroup=false
    Override Seed=false
    Rotation Mode=::NewEnumerator1
    Sphere Radius=&MergeHelper_FloatSwitch(0=20, 1=50, 2=10, 3=?, Index=$Emitter.SpawnGroup, Num=::NewEnumerator1)
    UseOwnerScale=true
   )
  ]
  particleUpdateScript [
   /SetVariables(Particles.VisibilityTag=&SpawnGroup()) {isSetParameters=true}
   &ParticleState()
   &InheritSourceMovement(
    Applied Position Scale=&Multiply_VectorByFloat(
     Vector=<1 1 1>
     Float=&FloatFromCurve(FloatCurve=DataInterface() {propertyValues {Curve {keys [
          {arriveTangent=0.0056, interpMode=RCIM_Cubic, leaveTangent=0.0056, tangentMode=RCTM_User, time=0.1, value=1}
          {arriveTangent=-1.8464, interpMode=RCIM_Cubic, leaveTangent=-1.8464, tangentMode=RCTM_User, time=1, value=0.5}
         ]}}})
    )
    ApplyForce=false
    ApplyPosition=true
    ApplyPositionFalloff=false
   )
   &SolveForcesAndVelocity(Clamp Velocity=false, Limit Acceleration=false)
   &SubUVAnimation(
    "Sprite Renderer "=DataInterface() {propertyValues {SpriteRenderer=&self:Sprites001_0.NiagaraSpriteRendererProperties_1}}
    SubUV Lookup Index Scale:NiagaraFloat=&RandomRangeFloat(Minimum=-1, Maximum=1)
    UseEndFrame=false
    UseStartFrame=false
   )
   &SubUVAnimation(
    "Sprite Renderer "=DataInterface() {propertyValues {SpriteRenderer=&self:Sprites001_0.NiagaraSpriteRendererProperties_2}}
    SubUV Animation Mode:ENiagara_SubUVLookupModeV2=::NewEnumerator1
    SubUV Curve=DataInterface() {propertyValues {Curve {
       defaultValue=0
       keys [{interpMode=RCIM_Cubic, tangentMode=RCTM_User, time=0, value=0}, {interpMode=RCIM_Cubic, tangentMode=RCTM_User, time=1, value=1}]
      }}}
    UseEndFrame=false
    UseStartFrame=false
   ) {module=SubUVAnimation001}
   /SetVariables(
    Particles.Marshmallow Size:Vector2f=&Vector2DFromFloat(Value=&FloatFromCurve(
      FloatCurve=DataInterface() {propertyValues {Curve {
         defaultValue=1
         keys [
          {arriveTangent=3.8153, interpMode=RCIM_Cubic, leaveTangent=3.8153, tangentMode=RCTM_User, time=0, value=0.5}
          {interpMode=RCIM_Cubic, tangentMode=RCTM_User, time=0.2801, value=1}
          {arriveTangent=-1.5641, tangentMode=RCTM_User, time=1, value=0.4}
         ]
        }}}
      Scale Curve=22
     ))
    Particles.Smoke Size:Vector2f=&Vector2DFromFloat(Value=&FloatFromCurve(
      FloatCurve=DataInterface() {propertyValues {Curve {
         defaultValue=1
         keys [
          {interpMode=RCIM_Cubic, tangentMode=RCTM_User, time=0, value=0.4}
          {arriveTangent=0.9035, interpMode=RCIM_Cubic, leaveTangent=0.9035, tangentMode=RCTM_User, time=0.8516, value=1}
         ]
        }}}
      Scale Curve=80
     ))
    Particles.Spark Size:Vector2f=&Vector2DFromFloat(Value=&RandomRangeFloat(Minimum=3.5, Maximum=4))
   ) {isSetParameters=true}
   /SetVariables(
    Particles.BC_Smoke:LinearColor=&MakeLinearColorFromVectorAndFloat(
     "Vector (RGB)"=&MakeVectorFromLinearColorRGB(Color=<rgba 1 1 1 1>)
     "Float (Alpha)"=&FloatFromCurve(FloatCurve=DataInterface() {propertyValues {Curve {
         defaultValue=1
         keys [{interpMode=RCIM_Cubic, tangentMode=RCTM_User, time=0, value=1}, {interpMode=RCIM_Cubic, tangentMode=RCTM_User, time=1, value=0}]
        }}})
    )
    Particles.BC_Spark:LinearColor=&MakeLinearColorFromVectorAndFloat(
     "Vector (RGB)"=&MakeVectorFromLinearColorRGB(Color=<rgba 1 0.7454 0.5333 1>)
     "Float (Alpha)"=&FloatFromCurve(FloatCurve=DataInterface() {propertyValues {Curve {
         defaultValue=1
         keys [
          {interpMode=RCIM_Cubic, tangentMode=RCTM_User, time=0, value=1}
          {arriveTangent=-3.7133, interpMode=RCIM_Cubic, leaveTangent=-3.7133, tangentMode=RCTM_User, time=1, value=0}
         ]
        }}})
    )
   ) {isSetParameters=true}
  ]
  rendererClasses [&NiagaraSpriteRendererProperties]
  renderers [
   {rendererClass=&NiagaraSpriteRendererProperties}
   {rendererClass=&NiagaraSpriteRendererProperties, rendererIndex=1}
   {rendererClass=&NiagaraSpriteRendererProperties, rendererIndex=2}
  ]
 }
}
summary {
 systemName=NS_Boost_Cocoasmoore
 emitters [
  {emitterName=Ribbons, rendererClasses [&NiagaraRibbonRendererProperties]}
  {emitterName=Sprites001, rendererClasses [&NiagaraSpriteRendererProperties]}
 ]
 userVariables [
  User.bIsFrontendPreview {type=NiagaraBool}
  User.bIsPainted {type=NiagaraBool}
  User.ColorVariantID {type=NiagaraInt32}
  User.DTCarOffset {type=Vector3f, description="", defaultValue {struct=Vector3f, value=<0 0 0>}}
  User.GameModeID {type=NiagaraInt32}
  User.PaintedColor {type=LinearColor, description="", defaultValue {struct=LinearColor, value=<rgba 1 1 1 1>}}
  User.PaintedScalar {type=NiagaraFloat, defaultValue=1, description=""}
  User.PortCount {type=NiagaraInt32, defaultValue=2}
 ]
}
